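# Translucent unlit material for Niagara sprites in the Arcana Agate backpack FX, drawing a soft procedural ring (donut) with no textures.
# An outer exponential radial gradient is multiplied by the inverse of an inner one to cut the hole, then by particle alpha.
# Particle color RGB feeds the shared MF_FX_EmissiveAdjustments function.
# Radius and density of both gradients are scalar parameters; despite the RadialBlur name, no blur sampling occurs.
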Material /BRCosmetics/Accessories/FORT_Backpacks/Backpack_F_MED_ArcanaAgate/FX/Material/M_ArcanaAgate_RadialBlur_Donut
graph {
 properties {
  bAutomaticallySetUsageInEditor=false
  bUsedWithNiagaraSprites=true
  blendMode=BLEND_Translucent
  shadingModel=MSM_Unlit
  usedShadingModels=Unlit
 }
 nodes [
  n1 = ParticleColor()
  MaterialFunctionCall(Emissive=@n1.RGB) {materialFunction=&MF_FX_EmissiveAdjustments}
  n2 = TextureCoordinate()
  n3 = ScalarParameter() {defaultValue=0.5, parameterName=OuterRadius}
  n4 = ScalarParameter() {defaultValue=10, parameterName=OuterDensity}
  n5 = MaterialFunctionCall(
   UVs=TextureCoordinate()
   Radius=ScalarParameter() {defaultValue=0.8, parameterName=InnerRadius}
   Density=ScalarParameter() {defaultValue=3, parameterName=InnerDensity}
  ) {materialFunction=&RadialGradientExponential}
  Multiply(
   A=@n1.A
   B=Multiply(
    A=MaterialFunctionCall(UVs=@n2, Radius=@n3, Density=@n4) {materialFunction=&RadialGradientExponential}.RadialGradientExponential
    B=OneMinus(None=@n5.RadialGradientExponential)
   )
  )
 ]
}
