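# Niagara burst effect for the Arcana Agate backpack cosmetic, layering flares, a flash with a light, and sparkles.
# Per-frame spawns are gated on a value-state-change flag and skip the first frame.
# User.PictureID adds either flying bat meshes (0) or swirl meshes (1), with swirl and flares tracking the skeletal mesh.
# Separate sparkle emitters cover gameplay and front end, and time-of-day scalars adjust brightness.
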NiagaraSystem /BRCosmetics/Accessories/FORT_Backpacks/Backpack_F_MED_ArcanaAgate/FX/NS_Backpack_ArcanaAgate_BurstDuplicate
dependencies {
 usedDataInterfaces [
  &NiagaraDataInterfaceCurve
  &NiagaraDataInterfaceColorCurve
  &NiagaraDataInterfaceVector2DCurve
  &NiagaraDataInterfaceSkeletalMesh
 ]
 usedDynamicInputs []
 usedModules [
  &self:SystemSpawnScript.SystemScriptSource.SystemScriptGraph.NiagaraNodeAssignment_0.SetVariables_#1
  &self:SystemSpawnScript.SystemScriptSource.SystemScriptGraph.NiagaraNodeAssignment_1.SetVariables_#2
  &SystemState
  &self:SystemSpawnScript.SystemScriptSource.SystemScriptGraph.NiagaraNodeAssignment_4.SetVariables_#3
  /Game/Effects/Niagara/Modules/Logic/UpdateValueStateChange
  &Timeline
 ]
}
emitters {
 Bats {
  emitterUpdateScript [
   &EmitterState()
   &SpawnPerFrame(
    Spawn:NiagaraBool=&Boolean_AndOperation(
     Bool1=&Boolean_AndOperation(
      Bool1=&Boolean_NotOperation(Bool=$System.UpdateValueStateChange.Changed)
      Bool2=&Boolean_NotOperation(Bool=$Transient.FirstFrame)
     )
     Bool2=&SetBoolByIntComparison(A=$User.PictureID, B=0, Comparison Type=::NewEnumerator4)
    )
    Spawn Count=25
    Use Spawn Probability=false
   )
  ]
  particleSpawnScript [
   /SetVariables(Particles.Age:NiagaraFloat=&Lerp_Float(A=0, B=-0.5, Alpha=&ReturnNormalizedExecIndex())) {isSetParameters=true}
   /SetVariables(Particles.RandVel:NiagaraFloat=&FloatFromCurve(
     FloatCurve=DataInterface() {propertyValues {Curve {keys [{interpMode=RCIM_Cubic, tangentMode=RCTM_Break, time=0, value=0}, {interpMode=RCIM_Cubic, tangentMode=RCTM_Break, time=1, value=1}]}}}
     CurveIndex=&RandomRangeFloat(Minimum=0, Maximum=1)
    )) {isSetParameters=true}
   &InitializeParticle(
    Color=<rgba 0.9553 0 1 1>
    Lifetime Max=0.7
    Lifetime Min=0.3
    Lifetime Mode=::NewEnumerator1
    Mesh Scale Mode=::NewEnumerator1
    Mesh Uniform Scale Max=0.45
    Mesh Uniform Scale Min=0.25
    UsePositionOffset=false
    Write Material Random=true
   )
   &DynamicMaterialParameters(
    Index 0 Param 1=&RandomRangeFloat(Minimum=0, Maximum=3)
    Index 0 Param 2=&Lerp_Float(A=1, B=3, Alpha=$Particles.RandVel)
    Module.Param0WriteEnabled=true
    Module.Param1WriteEnabled=true
    Module.Param2WriteEnabled=false
    Module.Param3WriteEnabled=false
   )
   &ShapeLocation(
    MaskSpawnGroup=false
    Sphere Distribution=::NewEnumerator1
    Sphere Radius=10
    U Position=&RandomRangeFloat(Minimum=0.6, Maximum=0.9)
    UseOwnerScale=true
    V Position=&RandomRangeFloat(Minimum=0.2, Maximum=0.8)
   )
  ]
  particleUpdateScript [
   /Niagara/Modules/Update/Utility/VectorNoiseValue(Noise Frequency=30, Noise Quality / Cost:ENiagaraVectorNoiseQuality=::NewEnumerator0, Pan Noise=false)
   /SetVariables(Particles.Position=&AddVectorToPosition(
     Position=$Engine.Owner.Position
     Vector=&Add_Vector(
      A=&Lerp_Vector(
       A=<0 0 0>
       B=&Multiply_VectorByFloat(
        Vector=&TransformVector(
         Vector=&RandomRangeVector(Minimum=<-0.5 -1.1 -0.5>, Maximum=<0.5 -1.1 0.5>)
         Destination Space=::Simulation
        )
        Float=75
       )
       Alpha=&FloatFromCurve(FloatCurve=DataInterface() {propertyValues {Curve {defaultValue=0, keys [{tangentMode=RCTM_User, time=0, value=0}, {tangentMode=RCTM_User, time=1, value=1}]}}})
      )
      B=&Multiply_VectorByFloat(
       Vector=<0 0 30>
       Float=&FloatFromCurve(FloatCurve=DataInterface() {propertyValues {Curve {keys [
            {arriveTangent=-0.0885, interpMode=RCIM_Cubic, leaveTangent=-0.0885, tangentMode=RCTM_User, time=0, value=0}
            {arriveTangent=2.6396, interpMode=RCIM_Cubic, leaveTangent=2.6396, tangentMode=RCTM_User, time=1, value=1}
           ]}}})
      )
     )
    )) {isSetParameters=true}
   &SolveForcesAndVelocity(Clamp Velocity=false, Limit Acceleration=false)
   &ParticleState()
   &ScaleMeshSize(Scale Factor=&VectorFromFloat(Value=&FloatFromCurve(FloatCurve=DataInterface() {propertyValues {Curve {keys [
          {interpMode=RCIM_Cubic, leaveTangent=9.1542, tangentMode=RCTM_Break, time=0, value=0}
          {interpMode=RCIM_Cubic, tangentMode=RCTM_Break, time=0.1, value=1}
          {arriveTangent=-1.8807, interpMode=RCIM_Cubic, tangentMode=RCTM_Break, time=1, value=0.1}
         ]}}})))
   /SetVariables(Particles.MeshOrientation=&OrientMeshToVector(
     Look At Direction=&NormalizeVector(
      Vector To Normalize=&Subtract_Position(A=$Particles.Position, B=$Engine.Owner.Position)
      Safe Normalize=true
      Safe Fallback Threshold=0
      Safe Fallback Vector=<0 0 0.001>
     )
     Reference Axis Vector=<1 0 0>
    )) {isSetParameters=true}
  ]
  rendererClasses [&NiagaraMeshRendererProperties]
  renderers [{rendererClass=&NiagaraMeshRendererProperties, rendererIndex=0}]
 }
 Flares {
  emitterUpdateScript [
   &EmitterState()
   &SpawnPerFrame(
    Spawn:NiagaraBool=&Boolean_AndOperation(
     Bool1=&Boolean_NotOperation(Bool=$System.UpdateValueStateChange.Changed)
     Bool2=&Boolean_NotOperation(Bool=$Transient.FirstFrame)
    )
    Use Spawn Probability=false
   )
  ]
  particleSpawnScript [
   &InitializeParticle(
    Lifetime=4
    Mesh Scale Mode=::NewEnumerator3
    Mesh Uniform Scale=1.7
    Position Offset=<0 3 0>
    Sprite Size Mode=::NewEnumerator3
    Uniform Sprite Size=80
    UsePositionOffset=false
    Write Material Random=true
   )
   &InitialMeshOrientation(Mesh Axis To Orient=<0 1 0>, Mesh Orientation Mode=::NewEnumerator2, Rotation=<-0.45 0 0>, Use Rotation Vector=true)
   /SetVariables(
    Particles.RainbowSize:Vector2f=&Vector2DFromFloat(Value=70)
    Particles.ShinyEdgeCameraOffset=5
    Particles.ShinyEdgeSize:Vector2f=&Vector2DFromFloat(Value=18)
   ) {isSetParameters=true}
  ]
  particleUpdateScript [
   &ParticleState()
   /SetVariables(
    Particles.RainbowDynamic02:Vector4f=&MakeVector4(X=1, Y=0.2, Z=0.7, W=1)
    Particles.ShinyEdgePosition:NiagaraPosition=&ConvertVectorToPosition(Input Position=<0 -6.5 2>, Conversion Operation:ENiagara_LWCConvertVecToPos=::NewEnumerator0)
   ) {isSetParameters=true}
   &ScaleSpriteSize(Scale Factor=&Multiply_Vector2D_ByFloat(
     Float=&FloatFromCurve(
      FloatCurve=DataInterface() {propertyValues {Curve {defaultValue=1.4424, keys [{tangentMode=RCTM_User, time=0, value=0}, {tangentMode=RCTM_User, time=1, value=1}]}}}
      CurveIndex=$Engine.System.Age
     )
     Vector2D=&Vector2DFromFloat(Value=&MakeFloatFromVector2d(Vector 2D=&Vector2DFromFloat(Value=&Sine(Normalized Angle=$Emitter.Age, Period=1, Scale=0.1, Bias=1))))
    ))
   &DynamicMaterialParameters(
    Index 0 Param 1=0.5
    Index 0 Param 4=&Waveform("Waveform [1]"=::NewEnumerator12, "[1] Frequency"=1, "[1] Amplitude Min/Max"=<0.5 1>)
    Module.Index0Param0Write=true
    Module.Index0Param1Write=true
    Module.Index0Param2Write=true
    Module.Index0Param3Write=false
   )
   &SkeletalMeshLocation(
    MaskSpawnGroup=false
    Mesh Sampling Type=::NewEnumerator5
    Sampled Position Offset=<0 -1.5 0>
    Skeletal Mesh=$System.Skeletal Mesh
    Socket Sampling Mode=::NewEnumerator5
   )
   &FN_Mesh_Attributes(
    Color=<rgba 2 0.8941 0.5382 0.4>
    Create Color Attribute=true
    Create Orientation Attribute=true
    Create Scale Attribute=true
    Orientation:Quat4f=&MakeQuaternion(Quaternion From=0, Axis=<0 1 0>, Angle=90, From=?, To=?, Angle Type=::NewEnumerator0, Normalize Vectors=false)
    Scale=&VectorFromFloat(Value=&Multiply_Float(A=8.2, B=$User.ScaleM))
    Write to Module Namespace=true
   )
   /SetVariables(Particles.FloatingFragColor:LinearColor=&MultiplyLinearColor(
     Linear Color=<rgba 5 1.6061 0.6377 1>
     Scale Factor=&MakeFloat4FromVectorAndFloat(
      "Vector (RGB)"=&Multiply_VectorByFloat(
       Vector=<1 1 1>
       Float=&FloatFromCurve(
        FloatCurve=DataInterface() {propertyValues {Curve {
           defaultValue=0.4234
           keys [
            {arriveTangent=0.0486, interpMode=RCIM_Cubic, leaveTangent=0.0486, tangentMode=RCTM_User, time=0, value=3}
            {arriveTangent=0.0213, interpMode=RCIM_Cubic, leaveTangent=0.0213, tangentMode=RCTM_User, time=0.0966, value=1}
            {interpMode=RCIM_Cubic, tangentMode=RCTM_User, time=1, value=0}
           ]
          }}}
        Scale Curve=2
       )
      )
      "Float (Alpha)"=&FloatFromCurve(
       FloatCurve=DataInterface() {propertyValues {Curve {
          defaultValue=0.4234
          keys [
           {interpMode=RCIM_Cubic, tangentMode=RCTM_User, time=0, value=1}
           {arriveTangent=-6.6662, interpMode=RCIM_Cubic, leaveTangent=-6.6662, tangentMode=RCTM_User, time=0.1, value=0.3792}
           {arriveTangent=5.5428, interpMode=RCIM_Cubic, leaveTangent=5.5428, tangentMode=RCTM_User, time=0.4245, value=0.3696}
           {arriveTangent=-0.0323, interpMode=RCIM_Cubic, leaveTangent=0.1524, tangentMode=RCTM_Break, time=1, value=0}
          ]
         }}}
       Scale Curve=&FNTimeOfDayScalar(Night Scale=0.4, Morning Scale=1, Evening Scale=0.4)
      )
     )
    )) {isSetParameters=true}
   /SetVariables(Particles.GlowColor=&MultiplyLinearColor(
     Linear Color=<rgba 3 1.7846 0.4322 0.8>
     Scale Factor=&MakeFloat4FromVectorAndFloat(
      "Vector (RGB)"=<1 1 1>
      "Float (Alpha)"=&FloatFromCurve(FloatCurve=DataInterface() {propertyValues {Curve {
          defaultValue=1
          keys [
           {arriveTangent=-9.7533, interpMode=RCIM_Cubic, leaveTangent=-9.7533, tangentMode=RCTM_User, time=0.7015, value=1}
           {interpMode=RCIM_Cubic, tangentMode=RCTM_User, time=1, value=0}
          ]
         }}})
     )
    )) {isSetParameters=true}
   /SetVariables(Particles.RainbowColor:LinearColor=&MultiplyLinearColor(
     Linear Color=<rgba 1 0.5 0.0625 1>
     Scale Factor=&MakeFloat4FromVectorAndFloat(
      "Vector (RGB)"=&VectorFromFloat(Value=2)
      "Float (Alpha)"=&FloatFromCurve(
       FloatCurve=DataInterface() {propertyValues {Curve {
          defaultValue=1
          keys [
           {arriveTangent=9.0725, interpMode=RCIM_Cubic, leaveTangent=9.0725, tangentMode=RCTM_User, time=0, value=0}
           {arriveTangent=-1.0007, interpMode=RCIM_Cubic, leaveTangent=-1.0007, tangentMode=RCTM_User, time=0.3676, value=0.3258}
           {arriveTangent=2.0356, interpMode=RCIM_Cubic, leaveTangent=2.0356, tangentMode=RCTM_User, time=0.5578, value=0.3775}
           {arriveTangent=-0.0179, interpMode=RCIM_Cubic, leaveTangent=-0.0179, tangentMode=RCTM_User, time=1, value=0}
          ]
         }}}
       Scale Curve=&FN_FrontEndFloatModifier(
        Gameplay=&FNTimeOfDayScalar(Night Scale=0.4, Morning Scale=0.8, Day Scale=0.8, Evening Scale=0.6)
        Front End=0.3
       )
      )
     )
    )) {isSetParameters=true}
   /SetVariables(Particles.GlowFacing:Vector3f=<0 -1 0>) {isSetParameters=true}
   /SetVariables(Particles.ShinyEdgeColor:LinearColor=&MultiplyLinearColor(
     Linear Color=<rgba 2 0.719 0.1665 1>
     Scale Factor=&MakeFloat4FromVectorAndFloat(
      "Vector (RGB)"=<1 1 1>
      "Float (Alpha)"=&FloatFromCurve(
       FloatCurve=DataInterface() {propertyValues {Curve {
          defaultValue=1
          keys [
           {arriveTangent=-6.092, interpMode=RCIM_Cubic, leaveTangent=-6.092, tangentMode=RCTM_User, time=0, value=2}
           {arriveTangent=-0.1567, interpMode=RCIM_Cubic, leaveTangent=-0.1567, tangentMode=RCTM_User, time=1, value=0}
          ]
         }}}
       Scale Curve=&FNTimeOfDayScalar(Night Scale=0.4, Morning Scale=0.8, Day Scale=0.8, Evening Scale=0.6)
      )
     )
    )) {isSetParameters=true}
   &ScaleColor(
    Scale RGB=&MakeVectorFromLinearColorRGB(Color=&ColorFromCurve(ColorCurve=DataInterface() {propertyValues {
        AlphaCurve {keys [<at 0 1 RCIM_Cubic>,<at 1 0 RCIM_Cubic>]}
        BlueCurve {keys [<at 0 1 RCIM_Cubic>,<at 1 1 RCIM_Cubic>]}
        GreenCurve {keys [<at 0 1 RCIM_Cubic>,<at 1 1 RCIM_Cubic>]}
        RedCurve {keys [<at 0 1 RCIM_Cubic>,<at 1 1 RCIM_Cubic>]}
       }}))
    ScaleA=true
    ScaleRGB=true
   )
  ]
  rendererClasses [&NiagaraMeshRendererProperties, &NiagaraSpriteRendererProperties]
  renderers [
   {rendererClass=&NiagaraMeshRendererProperties, rendererIndex=0}
   {rendererClass=&NiagaraSpriteRendererProperties, rendererIndex=1}
   {rendererClass=&NiagaraSpriteRendererProperties, rendererIndex=2}
   {rendererClass=&NiagaraSpriteRendererProperties, rendererIndex=3}
  ]
 }
 Flash {
  emitterUpdateScript [
   &EmitterState()
   &SpawnPerFrame(
    Spawn:NiagaraBool=&Boolean_AndOperation(
     Bool1=&Boolean_NotOperation(Bool=$System.UpdateValueStateChange.Changed)
     Bool2=&Boolean_NotOperation(Bool=$Transient.FirstFrame)
    )
    Use Spawn Probability=false
   )
  ]
  particleSpawnScript [
   &InitializeParticle(
    Color=<rgba 2 0.9516 0.1715 1>
    Lifetime=0.3
    Sprite Rotation Angle=110
    Sprite Rotation Mode=::NewEnumerator0
    Sprite Size Mode=::NewEnumerator3
    Uniform Sprite Size=&FN_FrontEndFloatModifier(Gameplay=22, Front End=14)
    UsePositionOffset=true
    Write Material Random=true
   )
   &SpriteFacingAndAlignment(Sprite Facing=<0 -1 0>, WriteAlignment=false, WriteFacing=true)
   &CameraOffset(Camera Offset Amount=5)
  ]
  particleUpdateScript [
   &ParticleState()
   &ScaleSpriteSize(Uniform Curve Sprite Scale=DataInterface() {propertyValues {Curve {keys [
        {arriveTangent=5.2794, interpMode=RCIM_Cubic, leaveTangent=5.2794, tangentMode=RCTM_User, time=0, value=1}
        {arriveTangent=0.7472, interpMode=RCIM_Cubic, leaveTangent=0.7472, tangentMode=RCTM_User, time=1, value=4}
       ]}}})
   &DynamicMaterialParameters(
    Index 0 Param 1=1
    Index 0 Param 2=&Lerp_Float(
     A=0.7
     B=0
     Alpha=&FloatFromCurve(FloatCurve=DataInterface() {propertyValues {Curve {keys [
          {arriveTangent=1.3124, interpMode=RCIM_Cubic, leaveTangent=1.3124, tangentMode=RCTM_User, time=0, value=0}
          {arriveTangent=0.3903, interpMode=RCIM_Cubic, leaveTangent=0.3903, tangentMode=RCTM_User, time=1, value=1}
         ]}}})
    )
    Module.Index0Param0Write=true
    Module.Index0Param1Write=true
    Module.Index0Param2Write=true
    Module.Index0Param3Write=false
   )
   &ScaleColor(
    Scale Alpha=&FloatFromCurve(
     FloatCurve=DataInterface() {propertyValues {Curve {keys [<at 0 1 RCIM_Cubic>,<at 1 0>]}}}
     Scale Curve=&FNTimeOfDayScalar(Night Scale=0.8, Morning Scale=1, Evening Scale=1)
    )
    Scale RGB=&VectorFromFloat(Value=&FloatFromCurve(
      FloatCurve=DataInterface() {propertyValues {Curve {
         defaultValue=1
         keys [
          {arriveTangent=0.102, interpMode=RCIM_Cubic, leaveTangent=0.102, tangentMode=RCTM_User, time=0.1408, value=0.9847}
          {arriveTangent=-0.0405, interpMode=RCIM_Cubic, leaveTangent=-0.0405, tangentMode=RCTM_User, time=1, value=0}
         ]
        }}}
      Scale Curve=2
     ))
    ScaleA=true
    ScaleRGB=true
   )
   /SetVariables(
    Particles.Darken Color:LinearColor=&MakeLinearColorFromVectorAndFloat(
     "Vector (RGB)"=&MakeVectorFromLinearColorRGB(Color=<rgba 0.01 0.0143 0.02 1>)
     "Float (Alpha)"=&FloatFromCurve(FloatCurve=DataInterface() {propertyValues {Curve {keys [{interpMode=RCIM_Cubic, tangentMode=RCTM_User, time=0, value=1}, {interpMode=RCIM_Cubic, tangentMode=RCTM_User, time=1, value=0}]}}})
    )
    Particles.Darken Size:Vector2f=&Multiply_Vector2D_ByFloat(Float=1, Vector2D=$Particles.SpriteSize)
   ) {isSetParameters=true}
   /SetVariables(
    Particles.Glow Camera Offset=5
    Particles.Glow Color=&MultiplyLinearColor(
     Linear Color=$Particles.Color
     Scale Factor=&MakeFloat4FromVectorAndFloat(
      "Vector (RGB)"=&VectorFromFloat(Value=1)
      "Float (Alpha)"=&FloatFromCurve(
       FloatCurve=DataInterface() {propertyValues {Curve {keys [
           {arriveTangent=3.5878, interpMode=RCIM_Cubic, leaveTangent=3.5878, tangentMode=RCTM_User, time=0, value=0.35}
           {arriveTangent=0.3, interpMode=RCIM_Cubic, leaveTangent=0.3, time=0.3, value=1}
           {arriveTangent=3.4933, interpMode=RCIM_Cubic, leaveTangent=3.4933, tangentMode=RCTM_User, time=0.5, value=0.5}
           {arriveTangent=-0.4, arriveTangentWeight=0.0807, interpMode=RCIM_Cubic, leaveTangent=-0.4, leaveTangentWeight=0.3549, time=0.6, value=0.75}
           {arriveTangent=-0.0405, interpMode=RCIM_Cubic, leaveTangent=-0.0405, tangentMode=RCTM_User, time=1, value=0.3}
          ]}}}
       Scale Curve=0.05
      )
     )
    )
    Particles.Glow Size=&Vector2DFromFloat(Value=50)
   ) {isSetParameters=true}
   /SetVariables(
    Particles.Wide Flare Position:NiagaraPosition=&ConvertPositionToVector(Input Position=<-0.5 -10 2.5>, Conversion Operation:ENiagara_LWCConvertPosToVec=::NewEnumerator0)
    Particles.Wide Flare Scale:Vector2f=&Vector2DFromCurve(
     Vector2Curve=DataInterface() {propertyValues {
       XCurve {
        defaultValue=1
        keys [
         {interpMode=RCIM_Cubic, tangentMode=RCTM_User, time=0, value=0.8}
         {interpMode=RCIM_Cubic, tangentMode=RCTM_User, time=0.12, value=1}
         {tangentMode=RCTM_User, time=1, value=0}
        ]
       }
       YCurve {
        defaultValue=1
        keys [
         {interpMode=RCIM_Cubic, tangentMode=RCTM_User, time=0, value=0.8}
         {interpMode=RCIM_Cubic, tangentMode=RCTM_User, time=0.12, value=1}
         {tangentMode=RCTM_User, time=1, value=0}
        ]
       }
      }}
     ScaleCurve=&Multiply_Vector2D_ByFloat(Float=2.5, Vector2D=$Particles.SpriteSize)
    )
   ) {isSetParameters=true}
   &FN_Light_Attributes(
    Color=<rgba 1 0.4173 0.1927 1>
    Create Color Attribute=true
    Create Exponent Attribute=true
    Create Radius Attribute=true
    Exponent=2
    Radius=55
    Scale Color=&FloatFromCurve(
     FloatCurve=DataInterface() {propertyValues {Curve {
        defaultValue=1
        keys [{interpMode=RCIM_Cubic, tangentMode=RCTM_User, time=0, value=1}, {interpMode=RCIM_Cubic, tangentMode=RCTM_User, time=1, value=0}]
       }}}
     Scale Curve=10
    )
   )
   /SetVariables(Particles.LightPosition:NiagaraPosition=&AddVectorToPosition(Position=$Particles.Position, Vector=<0 -11 0>)) {isSetParameters=true}
  ]
  rendererClasses [&NiagaraSpriteRendererProperties, &NiagaraLightRendererProperties]
  renderers [
   {rendererClass=&NiagaraSpriteRendererProperties}
   {rendererClass=&NiagaraSpriteRendererProperties, rendererIndex=1}
   {rendererClass=&NiagaraSpriteRendererProperties, rendererIndex=2}
   {rendererClass=&NiagaraLightRendererProperties, rendererIndex=3}
  ]
 }
 Sparkles {
  emitterUpdateScript [
   &EmitterState()
   &SpawnPerFrame(
    Spawn:NiagaraBool=&Boolean_AndOperation(
     Bool1=&Boolean_NotOperation(Bool=$System.UpdateValueStateChange.Changed)
     Bool2=&Boolean_NotOperation(Bool=$Transient.FirstFrame)
    )
    Spawn Count=15
    Spawn Probability:NiagaraFloat=&FN_FrontEndFloatModifier(Gameplay=1, Front End=0)
    Use Spawn Probability=true
   )
  ]
  particleSpawnScript [
   /SetVariables(Particles.Age:NiagaraFloat=&Lerp_Float(A=0, B=-3, Alpha=&ReturnNormalizedExecIndex())) {isSetParameters=true, module=SetVariables_#4}
   &InitializeParticle(
    Color=<rgba 6 2.1704 0.6171 1>
    Lifetime Max=0.6
    Lifetime Min=0.3
    Lifetime Mode=::NewEnumerator1
    Position Offset=<0 8 0>
    Sprite Size Mode=::NewEnumerator3
    Uniform Sprite Size=18
    UsePositionOffset=true
    Write Material Random=true
   )
   &CameraOffset(Camera Offset Amount=15)
   &ShapeLocation(
    MaskSpawnGroup=false
    Non Uniform Scale=<1 1 0.15>
    Override Seed=false
    Rotation Axis=<0 1 0>
    Rotation Mode=::NewEnumerator1
    Sphere Radius=20
    Sphere Surface Distribution=0.75
    UseOwnerScale=true
   )
  ]
  particleUpdateScript [
   &ParticleState()
   &ScaleSpriteSize(
    Uniform Curve Scale=&Waveform("Waveform [1]"=::NewEnumerator9, "[1] Frequency"=12, "[1] Amplitude Min/Max"=<0 1>)
    Uniform Curve Sprite Scale=DataInterface() {propertyValues {Curve {keys [
        {arriveTangent=9.0725, interpMode=RCIM_Cubic, leaveTangent=9.0725, tangentMode=RCTM_User, time=0, value=0}
        {arriveTangent=-0.1942, leaveTangent=-0.1942, time=0.15, value=1}
        {arriveTangent=-1.1765, interpMode=RCIM_Cubic, leaveTangent=-1.1765, time=0.7, value=1}
        {interpMode=RCIM_Cubic, leaveTangent=-0.0405, time=1, value=0}
       ]}}}
   )
   &Drag(Drag=&RandomRangeFloat(Minimum=10, Maximum=12), Use Linear Drag=true, Use Rotational Drag=true)
   &InheritSourceMovement(
    Applied Position Scale=&Multiply_VectorByFloat(
     Vector=<1 1 1>
     Float=&FloatFromCurve(FloatCurve=DataInterface() {propertyValues {Curve {keys [
          {arriveTangent=-1.2628, interpMode=RCIM_Cubic, leaveTangent=-1.2628, tangentMode=RCTM_User, time=0, value=1}
          {interpMode=RCIM_Cubic, tangentMode=RCTM_User, time=1, value=0}
         ]}}})
    )
    ApplyForce=false
    ApplyPosition=true
    ApplyPositionFalloff=false
   )
   &SolveForcesAndVelocity(Clamp Velocity=false, Limit Acceleration=false)
   &ScaleColor(
    Scale Alpha=&FloatFromCurve(
     FloatCurve=DataInterface() {propertyValues {Curve {keys [<at 0 1>,<at 1 0>]}}}
     Scale Curve=&FNTimeOfDayScalar(Night Scale=0.6, Morning Scale=1, Evening Scale=0.8)
    )
    Scale RGB=&Multiply_VectorByFloat(
     Vector=<1 1 1>
     Float=&Waveform(
      "Waveform [1]"=::NewEnumerator9
      "[1] Phase Scale"=2
      "[1] Frequency"=&RandomRangeFloat(Minimum=10, Maximum=20)
      "[1] Amplitude Min/Max"=<0 1>
     )
    )
    ScaleA=true
    ScaleRGB=true
   )
  ]
  rendererClasses [&NiagaraSpriteRendererProperties]
  renderers [{rendererClass=&NiagaraSpriteRendererProperties}]
 }
 Sparkles_FE {
  emitterUpdateScript [
   &EmitterState()
   &SpawnPerFrame(
    Spawn:NiagaraBool=&Boolean_AndOperation(
     Bool1=&Boolean_NotOperation(Bool=$System.UpdateValueStateChange.Changed)
     Bool2=&Boolean_NotOperation(Bool=$Transient.FirstFrame)
    )
    Spawn Count=15
    Spawn Probability:NiagaraFloat=&FN_FrontEndFloatModifier(Gameplay=0, Front End=1)
    Use Spawn Probability=true
   )
  ]
  particleSpawnScript [
   /SetVariables(Particles.Age:NiagaraFloat=&Lerp_Float(A=0, B=-3, Alpha=&ReturnNormalizedExecIndex())) {isSetParameters=true, module=SetVariables_#4}
   &InitializeParticle(
    Color=<rgba 6 2.1704 0.6171 1>
    Lifetime Max=0.6
    Lifetime Min=0.3
    Lifetime Mode=::NewEnumerator1
    Position Offset=<0 -6 2>
    Position Offset Coordinate Space=::Local
    Sprite Size Mode=::NewEnumerator3
    Uniform Sprite Size=18
    UsePositionOffset=true
    Write Material Random=true
   )
   &CameraOffset(Camera Offset Amount=15)
   &ShapeLocation(
    MaskSpawnGroup=false
    Non Uniform Scale=<1 0.15 1>
    Override Seed=false
    Rotation Angle=-0.02
    Rotation Mode=::NewEnumerator1
    Sphere Radius=20
    Sphere Surface Distribution=0.75
    UseOwnerScale=true
   )
  ]
  particleUpdateScript [
   &ParticleState()
   &ScaleSpriteSize(
    Uniform Curve Scale=&Waveform("Waveform [1]"=::NewEnumerator9, "[1] Frequency"=12, "[1] Amplitude Min/Max"=<0 1>)
    Uniform Curve Sprite Scale=DataInterface() {propertyValues {Curve {keys [
        {arriveTangent=9.0725, interpMode=RCIM_Cubic, leaveTangent=9.0725, tangentMode=RCTM_User, time=0, value=0}
        {arriveTangent=-0.1942, leaveTangent=-0.1942, time=0.15, value=1}
        {arriveTangent=-1.1765, interpMode=RCIM_Cubic, leaveTangent=-1.1765, time=0.7, value=1}
        {interpMode=RCIM_Cubic, leaveTangent=-0.0405, time=1, value=0}
       ]}}}
   )
   &ScaleColor(
    Scale Alpha=&FloatFromCurve(
     FloatCurve=DataInterface() {propertyValues {Curve {keys [<at 0 1>,<at 1 0>]}}}
     Scale Curve=&FNTimeOfDayScalar(Night Scale=0.6, Morning Scale=1, Evening Scale=0.8)
    )
    Scale RGB=&Multiply_VectorByFloat(
     Vector=<1 1 1>
     Float=&Waveform(
      "Waveform [1]"=::NewEnumerator9
      "[1] Phase Scale"=2
      "[1] Frequency"=&RandomRangeFloat(Minimum=10, Maximum=20)
      "[1] Amplitude Min/Max"=<0 1>
     )
    )
    ScaleA=true
    ScaleRGB=true
   )
  ]
  rendererClasses [&NiagaraSpriteRendererProperties]
  renderers [{rendererClass=&NiagaraSpriteRendererProperties}]
 }
 SwirlMesh {
  emitterUpdateScript [
   &EmitterState()
   &SpawnPerFrame(
    Spawn:NiagaraBool=&Boolean_AndOperation(
     Bool1=&Boolean_AndOperation(
      Bool1=&Boolean_NotOperation(Bool=$System.UpdateValueStateChange.Changed)
      Bool2=&Boolean_NotOperation(Bool=$Transient.FirstFrame)
     )
     Bool2=&SetBoolByIntComparison(A=$User.PictureID, B=1, Comparison Type=::NewEnumerator4)
    )
    Spawn Count=2
    Use Spawn Probability=false
   )
  ]
  particleSpawnScript [&InitializeParticle(
    Color=<rgba 1 1 1 5>
    Lifetime=4
    Mesh Scale Mode=::NewEnumerator3
    Mesh Uniform Scale=0.1
    Position Offset=<0 -8 0>
    UsePositionOffset=true
    Write Material Random=true
   ) {module=InitializeParticle001}]
  particleUpdateScript [
   &ParticleState()
   &SkeletalMeshLocation(
    MaskSpawnGroup=false
    Mesh Sampling Type=::NewEnumerator5
    Sampled Position Offset=<0 -7 0>
    Skeletal Mesh=$System.Skeletal Mesh
    Socket Sampling Mode=::NewEnumerator5
   )
   &Color(
    Color=&ColorFromCurve(ColorCurve=DataInterface() {propertyValues {
       bOptimizeLUT=false
       AlphaCurve {keys [<at 0 1>]}
       BlueCurve {keys [<at 0 0>,<at 0.1114 0.0521>,<at 0.3018 0>,<at 0.3329 0.026>,<at 0.5221 0.0729>,<at 0.5915 0.0729>,<at 0.7289 0>,<at 0.75 0.026>,<at 0.8769 0.0781>,<at 0.9915 0.0677>]}
       GreenCurve {keys [<at 0 0>,<at 0.1114 0.0521>,<at 0.3018 0>,<at 0.3329 0.026>,<at 0.5221 0.0729>,<at 0.5915 0.0729>,<at 0.7289 0>,<at 0.75 0.026>,<at 0.8769 0.0781>,<at 0.9915 0.0677>]}
       RedCurve {keys [<at 0 0>,<at 0.1114 0.0521>,<at 0.3018 0>,<at 0.3329 0.026>,<at 0.5221 0.0729>,<at 0.5915 0.0729>,<at 0.7289 0>,<at 0.75 0.026>,<at 0.8769 0.0781>,<at 0.9915 0.0677>]}
      }})
    Scale Alpha=&FloatFromCurve(FloatCurve=DataInterface() {propertyValues {Curve {keys [
         {interpMode=RCIM_Cubic, tangentMode=RCTM_User, time=0, value=0}
         {interpMode=RCIM_Cubic, tangentMode=RCTM_User, time=0.35, value=1}
         {interpMode=RCIM_Cubic, tangentMode=RCTM_User, time=1, value=0}
        ]}}})
   )
   &UpdateMeshOrientation(Rotation Rate=-1, Rotation Vector=<0 1 0>)
   &SolveForcesAndVelocity(Clamp Velocity=false, Limit Acceleration=false)
   &DynamicMaterialParameters(
    Index 0 Param 1=&FloatFromCurve(
     FloatCurve=DataInterface() {propertyValues {Curve {keys [{interpMode=RCIM_Cubic, tangentMode=RCTM_User, time=0, value=0.25}, {interpMode=RCIM_Cubic, tangentMode=RCTM_User, time=1, value=2.5}]}}}
     Scale Curve=-75
    )
    Module.Param0WriteEnabled=false
    Module.Param1WriteEnabled=true
    Module.Param2WriteEnabled=true
    Module.Param3WriteEnabled=true
   )
   &CameraOffset()
   &ScaleColor(
    Scale Alpha=&FloatFromCurve(
     FloatCurve=DataInterface() {propertyValues {Curve {
        defaultValue=1
        keys [
         {arriveTangent=0.0598, interpMode=RCIM_Cubic, leaveTangent=0.0598, tangentMode=RCTM_User, time=0.002, value=0.0061}
         {arriveTangent=0.0166, interpMode=RCIM_Cubic, leaveTangent=0.0166, tangentMode=RCTM_User, time=0.0998, value=1.0022}
         {arriveTangent=-2.9996, interpMode=RCIM_Cubic, leaveTangent=-2.9996, tangentMode=RCTM_User, time=0.6507, value=0.3533}
         {arriveTangent=0.1176, interpMode=RCIM_Cubic, leaveTangent=0.1176, tangentMode=RCTM_User, time=1, value=0}
        ]
       }}}
     Scale Curve=0.1
    )
    Scale RGB=<0 0 0>
    ScaleA=true
    ScaleRGB=true
   )
  ]
  rendererClasses [&NiagaraMeshRendererProperties]
  renderers [{rendererClass=&NiagaraMeshRendererProperties, rendererIndex=0}]
 }
}
summary {
 systemName=NS_Backpack_ArcanaAgate_BurstDuplicate
 emitters [
  {emitterName=Bats, rendererClasses [&NiagaraMeshRendererProperties]}
  {emitterName=SwirlMesh, rendererClasses [&NiagaraMeshRendererProperties]}
  {emitterName=Sparkles, rendererClasses [&NiagaraSpriteRendererProperties]}
  {emitterName=Flash, rendererClasses [&NiagaraSpriteRendererProperties, &NiagaraLightRendererProperties]}
  {emitterName=Flares, rendererClasses [&NiagaraMeshRendererProperties, &NiagaraSpriteRendererProperties]}
  {emitterName=Sparkles_FE, rendererClasses [&NiagaraSpriteRendererProperties]}
 ]
 userVariables [
  User.bIsFrontEnd {type=NiagaraBool}
  User.bIsFrontEndPreview {type=NiagaraBool}
  User.PhysicalSurface {type=NiagaraInt32}
  User.PictureID {type=NiagaraInt32}
  User.RefreshInt {type=NiagaraInt32, defaultValue=-1}
  User.ScaleM {type=NiagaraFloat, defaultValue=0.045, description=""}
 ]
}
