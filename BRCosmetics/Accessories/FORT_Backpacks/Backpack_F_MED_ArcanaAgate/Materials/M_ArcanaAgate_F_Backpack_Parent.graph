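# Parent material for the ArcanaAgate backpack, built on the 2023 character PreFX/PostFX functions, adding a switchable picture in a UV1-masked region.
# PictureID blends between three quadrants of one picture texture: static Castle and Fred, and a continuously rotating Swirl.
# A PreviousFrameSwitch offsets only the previous frame's position, faking rotational motion vectors without moving geometry.
# That offset follows SpiralDirection and MotionVector_RotationRate, applying within the mask when Swirl shows.
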Material /BRCosmetics/Accessories/FORT_Backpacks/Backpack_F_MED_ArcanaAgate/Materials/M_ArcanaAgate_F_Backpack_Parent
graph {
 properties {
  bHasPixelAnimation=true
  bUseMaterialAttributes=true
  bUsedWithClothing=true
  bUsedWithMeshDeformer=true
  bUsedWithMorphTargets=true
  bUsedWithNanite=true
  bUsedWithSkeletalMesh=true
  pixelDepthOffsetMode=PDOM_Legacy
  previewMesh=/BRCosmetics/Accessories/FORT_Backpacks/Backpack_F_MED_ArcanaAgate/Meshes/SM_ArcanaAgate_Backpack
  shadingModel=MSM_FromMaterialExpression
  usedShadingModels=Default Lit | Subsurface | Clear Coat | Subsurface Profile | Cloth
 }
 nodes [
  n1 = MaterialFunctionCall() {materialFunction=&MF_CharacterParent_2023_PreFX}
  n2 = GetMaterialAttributes(None=@n1.Result) {attributeGetTypes [#1]}
  n3 = NamedRerouteUsage()
  n4 = MaterialFunctionCall(UVs=TextureCoordinate() {coordinateIndex=1}) {materialFunction=&ScaleUVsByCenter}
  n5 = Constant()
  n6 = NamedRerouteUsage()
  n7 = ScalarParameter() {defaultValue=1, group=01_ArcanaAgate_Reactivity, parameterName=PictureID, sliderMax=2}
  n8 = Min(A=@n4.0-1 mask, B=Fmod(A=@n7, B=Constant() {r=2}))
  MaterialFunctionCall(
   Material=SetMaterialAttributes!(
    MaterialAttributes=@n2.MaterialAttributes
    Base Color=LinearInterpolate(A=@n2.Base Color, B=@n3, Alpha=@n4.0-1 mask)
    World Position Offset=LinearInterpolate(A=@n5, B=@n6, Alpha=@n8)
   ) {attributeSetTypes [#1, #2]}
   M=@n1.M
   Pre Skinned Local Position=@n1.M
   Distance Blend=@n1.M
   Fresnel=@n1.M
  ) {materialFunction=&MF_CharacterParent_2023_PostFX}
  n9 = Multiply(A=TextureCoordinate() {coordinateIndex=1}) {constB=0.5}
  n10 = Constant2Vector() {r=0.5}
  n11 = LinearInterpolate(
   A=Rotator(Coordinate=@n9, Time=Constant() {r=1}) {centerX=0.25, centerY=0.25, desc=Castle, speed=1}
   B=Rotator(Coordinate=Add(A=@n9, B=@n10) {constB=0.5}) {centerX=0.75, centerY=0.25, desc=Swirl, speed=-2}
   Alpha=Saturate(None=@n7)
  )
  n12 = Rotator(Coordinate=Add(A=@n9, B=Constant2Vector() {g=0.5}) {constB=0.5}, Time=Constant() {r=1}) {centerX=0.25, centerY=0.75, desc=Fred, speed=1}
  n13 = Saturate(None=Subtract(A=@n7))
  NamedRerouteDeclaration(None=TextureSample(UVs=LinearInterpolate(A=@n11, B=@n12, Alpha=@n13)) {texture=/BRCosmetics/Accessories/FORT_Backpacks/Backpack_F_MED_ArcanaAgate/Textures/T_ArcanaAgate_Backpack_Pictures_D}.RGB) {name=Picture, nodeColor=<rgba 1 0 0.1641 1>}
  n14 = ScalarParameter() {defaultValue=0.5, group=01_ArcanaAgate_Reactivity, parameterName=SpiralDirection, sliderMax=1}
  n15 = Multiply(A=Rotator(
    Coordinate=TextureCoordinate() {coordinateIndex=1}
    Time=MaterialFunctionCall(Multiplier=@n14) {materialFunction=&Pi}.Result
   ) {speed=1}) {constB=2}
  MaterialExpressionMultiply_1 = Multiply(
   A=Multiply(A=Subtract(A=@n15), B=DeltaTime())
   B=ScalarParameter() {defaultValue=1, group=01_ArcanaAgate_Reactivity, parameterName=MotionVector_RotationRate, sliderMax=1}
  )
  n16 = Convert(None=@MaterialExpressionMultiply_1) {
   nodeName=Make Float3
   convertInputs [{
     type=Vector2
     expressionInput {
      expression=&self:MaterialExpressionMultiply_1
      inputName=None
      mask=0
      maskA=0
      maskB=0
      maskG=0
      maskR=0
      outputIndex=0
     }
    }]
   convertOutputs [{type=Vector3}]
  }
  NamedRerouteDeclaration(None=PreviousFrameSwitch(Current Frame=Constant(), Previous Frame=Transform(None=@n16))) {name=WPO, nodeColor=<rgba 0 0.1328 1 1>}
 ]
}
