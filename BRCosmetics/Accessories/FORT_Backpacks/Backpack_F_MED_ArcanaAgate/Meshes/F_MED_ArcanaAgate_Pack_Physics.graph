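# Physics asset for the F_MED_ArcanaAgate backpack cosmetic, defining collision and joints for its carabiner and hanging string.
# Capsule bodies form one chain from root through pack, base carabiner and dynamic carabiner to five string segments, with three larger capsules on root.
# The pack and base carabiner joints rotate freely, while the dynamic joints have angular limits, with tight or zero twist and wider swing on the last three string segments.
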PhysicsAsset /BRCosmetics/Accessories/FORT_Backpacks/Backpack_F_MED_ArcanaAgate/Meshes/F_MED_ArcanaAgate_Pack_Physics
bodies {
 base_carabiner [{shapeType=Capsule, length=0.505, radius=0.505, shapeName=base_carabiner_capsule, rotation=<rot 0 -90 0>}]
 dyn_carabiner [{
   shapeType=Capsule
   length=2.224
   radius=1.37
   shapeName=dyn_carabiner_capsule
   center=<1.0016 0.1285 0.7483>
   rotation=<rot 0 -93.5954 -78.2523>
  }]
 dyn_string [{shapeType=Capsule, length=3, radius=3, shapeName=dyn_string_capsule, center=<2 0 0>, rotation=<rot 0 90 90>}]
 dyn_string_01 [{shapeType=Capsule, length=3, radius=3, shapeName=dyn_string_01_capsule, center=<2 0 0>, rotation=<rot 0 90 90>}]
 dyn_string_02 [{shapeType=Capsule, length=3, radius=3, shapeName=dyn_string_02_capsule, center=<3.6447 0 0>, rotation=<rot 0 90 90>}]
 dyn_string_03 [{shapeType=Capsule, length=3, radius=3, shapeName=dyn_string_03_capsule, center=<5.3098 0 0>, rotation=<rot 0 90 90>}]
 dyn_string_04 [{shapeType=Capsule, length=3, radius=3, shapeName=dyn_string_04_capsule, center=<7.7277 0 0>, rotation=<rot 0 90 90>}]
 pack [{
   shapeType=Capsule
   length=30.1435
   radius=3.4592
   shapeName=pack_capsule
   center=<4.8543 -2.6599 -11.8576>
   rotation=<rot -0.1373 -90 -31.9501>
  }]
 root [
  {shapeType=Capsule, length=32, radius=13, shapeName=root_capsule, center=<0 15.5365 -13.8424>, rotation=<rot 90 0 0>}
  {shapeType=Capsule, length=15, radius=15, shapeName=root_capsule, center=<0 18.9373 -28.8866>, rotation=<rot 90 0 0>}
  {shapeType=Capsule, length=15, radius=15, shapeName=root_capsule, center=<-36.6787 0 10.4807>, rotation=<rot 90 -90 0>}
 ]
}
constraints [
 {
  bone1Name=base_carabiner
  bone2Name=pack
  swing1LimitDegrees=45
  swing1Motion=Free
  swing2LimitDegrees=45
  swing2Motion=Free
  twistLimitDegrees=45
  twistMotion=Free
 }
 {bone1Name=dyn_carabiner, bone2Name=base_carabiner, swing1LimitDegrees=10, swing2LimitDegrees=10, twistLimitDegrees=0}
 {bone1Name=dyn_string, bone2Name=dyn_carabiner, swing1LimitDegrees=30, swing2LimitDegrees=0, twistLimitDegrees=0}
 {bone1Name=dyn_string_01, bone2Name=dyn_string, swing1LimitDegrees=20, swing2LimitDegrees=0, twistLimitDegrees=0}
 {bone1Name=dyn_string_02, bone2Name=dyn_string_01, swing1LimitDegrees=40, swing2LimitDegrees=30, twistLimitDegrees=2}
 {bone1Name=dyn_string_03, bone2Name=dyn_string_02, swing1LimitDegrees=40, swing2LimitDegrees=30, twistLimitDegrees=2.5}
 {bone1Name=dyn_string_04, bone2Name=dyn_string_03, swing1LimitDegrees=40, swing2LimitDegrees=30, twistLimitDegrees=2}
 {
  bone1Name=pack
  bone2Name=root
  swing1LimitDegrees=45
  swing1Motion=Free
  swing2LimitDegrees=45
  swing2Motion=Free
  twistLimitDegrees=45
  twistMotion=Free
 }
]
