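# Anim Blueprint for the ArcanaAgate backpack, adding physics sway and bone offsets to its hinge, carabiner and strings.
# Two RigidBody sims relative to the pack bone, using normal and skydive physics assets, are blended by skydive-floating state.
# The cached result takes a pack offset derived from the owner's spine yaw, then fixed additive bone tweaks.
# The EventGraph also sets timed chest-open flags, which the AnimGraph shown does not read.
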AnimBlueprint /BRCosmetics/Accessories/FORT_Backpacks/Backpack_F_MED_ArcanaAgate/Meshes/F_MED_ArcanaAgate_Pack_AnimBP
target_skeleton=&/F_MED_ArcanaAgate_Pack_Skeleton
graphs [
 AnimGraph {nodes [
   n1 = "Transform (Modify) Bone - Bone: pack"(
    ComponentPose=Local To Component(LocalPose=Use cached pose 'Rb'())
    Translation=Get BP Translate() -> (BP Translate:Vector)
    Rotation=Get BP Rotate() -> (BP Rotate:Rotator)
   ) {
    rotationMode=BMM_Additive
    rotationSpace=BCS_BoneSpace
    translationMode=BMM_Additive
    translationSpace=BCS_BoneSpace
    alphaBoolBlend {blendOption=Linear}
    alphaScaleBiasClamp {bInterpResult=true}
    boneToModify {boneName=pack}
    rotation {roll=10}
   }
   n2 = "Transform (Modify) Bone - Bone: base_hinge"(ComponentPose=@n1, Alpha=1, Translation="0,0,0", Rotation="0,-115,0", Scale="1,1,1") {
    rotationMode=BMM_Additive
    rotationSpace=BCS_BoneSpace
    alphaBoolBlend {blendOption=Linear}
    alphaScaleBiasClamp {bInterpResult=true, interpSpeedDecreasing=5, interpSpeedIncreasing=5}
    boneToModify {boneName=base_hinge}
    rotation {yaw=-115}
   }
   n3 = "Transform (Modify) Bone - Bone: base_hinge"(ComponentPose=@n2, Alpha=1, Translation="0,0,0", Rotation="0,-115,0", Scale="1,1,1") {
    rotationMode=BMM_Additive
    rotationSpace=BCS_BoneSpace
    alphaBoolBlend {blendOption=Linear}
    alphaScaleBiasClamp {bInterpResult=true, interpSpeedDecreasing=5, interpSpeedIncreasing=5}
    boneToModify {boneName=base_hinge}
    rotation {yaw=-115}
   }
   n4 = "Transform (Modify) Bone - Bone: base_carabiner"(ComponentPose=@n3, Alpha=1, Translation="0,0,0", Rotation="0,0,-10", Scale="1,1,1") {
    rotationMode=BMM_Additive
    rotationSpace=BCS_BoneSpace
    translationMode=BMM_Additive
    translationSpace=BCS_BoneSpace
    alphaBoolBlend {blendOption=Linear}
    alphaScaleBiasClamp {bInterpResult=true, interpSpeedDecreasing=5, interpSpeedIncreasing=5}
    boneToModify {boneName=base_carabiner}
    rotation {roll=-10}
   }
   n5 = "Transform (Modify) Bone - Bone: dyn_string_04"(ComponentPose=@n4, Alpha=1, Translation="0,0,0", Rotation="-20,0,0", Scale="1,1,1") {
    rotationMode=BMM_Additive
    rotationSpace=BCS_BoneSpace
    translationMode=BMM_Additive
    translationSpace=BCS_BoneSpace
    alphaBoolBlend {blendOption=Linear}
    alphaScaleBiasClamp {bInterpResult=true, interpSpeedDecreasing=5, interpSpeedIncreasing=5}
    boneToModify {boneName=dyn_string_04}
    rotation {pitch=-20}
   }
   n6 = "Transform (Modify) Bone - Bone: dyn_string_02"(ComponentPose=@n5, Alpha=1, Translation="0,0,0", Rotation="-20,0,0", Scale="1,1,1") {
    rotationMode=BMM_Additive
    rotationSpace=BCS_BoneSpace
    translationMode=BMM_Additive
    translationSpace=BCS_BoneSpace
    alphaBoolBlend {blendOption=Linear}
    alphaScaleBiasClamp {bInterpResult=true, interpSpeedDecreasing=5, interpSpeedIncreasing=5}
    boneToModify {boneName=dyn_string_02}
    rotation {pitch=-20}
   }
   n7 = "Transform (Modify) Bone - Bone: dyn_string_03"(ComponentPose=@n6, Alpha=1, Translation="0,0,0", Rotation="-15,0,0", Scale="1,1,1") {
    rotationMode=BMM_Additive
    rotationSpace=BCS_BoneSpace
    translationMode=BMM_Additive
    translationSpace=BCS_BoneSpace
    alphaBoolBlend {blendOption=Linear}
    alphaScaleBiasClamp {bInterpResult=true, interpSpeedDecreasing=5, interpSpeedIncreasing=5}
    boneToModify {boneName=dyn_string_03}
    rotation {pitch=-15}
   }
   n8 = "Transform (Modify) Bone - Bone: dyn_string_01"(ComponentPose=@n7, Alpha=1, Translation="0.459,0.252,0.949", Rotation="-3.122,0.000,8.439", Scale="1,1,1") {
    rotationMode=BMM_Additive
    rotationSpace=BCS_BoneSpace
    translationMode=BMM_Additive
    translationSpace=BCS_BoneSpace
    alphaBoolBlend {blendOption=Linear}
    alphaScaleBiasClamp {bInterpResult=true, interpSpeedDecreasing=5, interpSpeedIncreasing=5}
    boneToModify {boneName=dyn_string_01}
    rotation=<rot -3.1219 0 8.439>
    translation=<0.459 0.252 0.9488>
   }
   n9 = "Transform (Modify) Bone - Bone: dyn_string_03"!(ComponentPose=@n8, bAlphaBoolEnabled:bool=True, Translation="0,0,0", Rotation="-30,0,0", Scale="1,1,1") -> (Pose) {
    alphaInputType=Bool
    rotationMode=BMM_Additive
    rotationSpace=BCS_BoneSpace
    translationMode=BMM_Additive
    translationSpace=BCS_BoneSpace
    alphaBoolBlend {blendInTime=0.5, blendOption=Linear, blendOutTime=1}
    alphaScaleBiasClamp {bInterpResult=true, interpSpeedDecreasing=5, interpSpeedIncreasing=5}
    boneToModify {boneName=dyn_string_03}
    rotation {pitch=-30}
   }
   Output Pose(Result=Component To Local(ComponentPose=@n9)) {layerGroup=None}
   Save cached pose 'start'(Pose=Input Pose()) {cacheName=start}
   n10 = RigidBody!(
    ComponentPose=Local To Component(LocalPose=Use cached pose 'start'())
    Alpha
    OverrideWorldGravity:Vector="0,0,-980"
    ExternalForce
    SimSpaceSettings:Sim Space Settings Structure=Get Sim Space Settings() -> (Sim Space Settings:Sim Space Settings Structure)
   ) -> (Pose) {
    bClampLinearTranslationLimitToRefPose=true
    bOverrideWorldGravity=true
    overridePhysicsAsset=&/F_MED_ArcanaAgate_Pack_Physics_skydive
    simulationSpace=BaseBoneSpace
    alphaBoolBlend {blendOption=Linear}
    baseBoneRef {boneName=pack}
    componentAppliedLinearAccClamp=<3000 3000 3000>
    componentLinearAccScale=<0.25 0.15 0.3>
    componentLinearVelScale=<0.25 0.15 0.3>
    overrideWorldGravity {z=-980}
    simSpaceSettings {
     maxAngularAcceleration=250
     maxAngularVelocity=50
     maxLinearAcceleration=400
     maxLinearVelocity=100
     velocityScaleZ=0
     worldAlpha=0.35
     externalLinearDragV=<0.2 0.2 0.2>
    }
   }
   n11 = "Transform (Modify) Bone - Bone: root"(ComponentPose=@n10, Translation="4,-2.7,0", Rotation="0,0,5") {
    rotationMode=BMM_Additive
    rotationSpace=BCS_BoneSpace
    translationMode=BMM_Additive
    translationSpace=BCS_BoneSpace
    alphaBoolBlend {blendOption=Linear}
    alphaScaleBiasClamp {bInterpResult=true}
    boneToModify {boneName=root}
    rotation {roll=5}
    translation=<4 -2.7>
   }
   n12 = Component To Local(ComponentPose=@n11)
   n13 = RigidBody(
    ComponentPose=Local To Component(LocalPose=Use cached pose 'start'())
    OverrideWorldGravity="0,0,-980"
    SimSpaceSettings=Get Sim Space Settings()
   ) {
    bClampLinearTranslationLimitToRefPose=true
    bOverrideWorldGravity=true
    overridePhysicsAsset=&/F_MED_ArcanaAgate_Pack_Physics
    simulationSpace=BaseBoneSpace
    alphaBoolBlend {blendOption=Linear}
    baseBoneRef {boneName=pack}
    componentAppliedLinearAccClamp=<3000 3000 3000>
    componentLinearAccScale=<0.25 0.15 0.3>
    componentLinearVelScale=<0.25 0.15 0.3>
    overrideWorldGravity {z=-980}
    simSpaceSettings {
     maxAngularAcceleration=250
     maxAngularVelocity=50
     maxLinearAcceleration=400
     maxLinearVelocity=100
     velocityScaleZ=0
     worldAlpha=0.35
     externalLinearDragV=<0.2 0.2 0.2>
    }
   }
   n14 = "Transform (Modify) Bone - Bone: root"(ComponentPose=@n13, Translation="4,-2.7,0", Rotation="0,0,5") {
    rotationMode=BMM_Additive
    rotationSpace=BCS_BoneSpace
    translationMode=BMM_Additive
    translationSpace=BCS_BoneSpace
    alphaBoolBlend {blendOption=Linear}
    alphaScaleBiasClamp {bInterpResult=true}
    boneToModify {boneName=root}
    rotation {roll=5}
    translation=<4 -2.7>
   }
   n15 = Component To Local(ComponentPose=@n14)
   n16 = Get bIsSkydiveFloating()
   n17 = "Transform (Modify) Bone - Bone: base_carabiner"!(
    ComponentPose=Local To Component(LocalPose=Blend Poses by bool(BlendPose_0=@n12, BlendPose_1=@n15, bActiveValue=@n16) {blendType=Linear})
    bAlphaBoolEnabled:bool=True
    Translation="0,0,0"
    Rotation="3,0,0"
    Scale="1,1,1"
   ) -> (Pose) {
    alphaInputType=Bool
    rotationMode=BMM_Additive
    rotationSpace=BCS_BoneSpace
    translationMode=BMM_Additive
    translationSpace=BCS_BoneSpace
    alphaBoolBlend {blendOption=Linear}
    alphaScaleBiasClamp {bInterpResult=true}
    boneToModify {boneName=base_carabiner}
    rotation {pitch=3}
   }
   Save cached pose 'Rb'(Pose=Component To Local(ComponentPose=@n17)) {cacheName=Rb}
  ]}
 EventGraph {nodes [
   Chestopen() {is=title K2Node_CustomEvent}
   > Set open(open:double=1) -> (Output_Get:double)
   > Delay(Duration=3)
   > Set open(open=0)
   Event BlueprintUpdateAnimation()
   > n1 = Sequence!(execute) -> (then_0, then_1, then_2:Exec, then_3:Exec)
   > n2 = Gravity Override State Function DEPRECATED(
    CharacterPartAnimInstance=Self-Reference()
    GameTimeInSeconds=GetGameTimeInSeconds()
    GravityOverrideParams=Get Gravity Override Params()
    RBAN Params=Get RBAN Params()
   )
   > Set Gravity Override(Gravity Override=@n2)
   > Set RBAN Settings Out!(execute, RBAN Settings Out:RBANSetting Struct Structure=@n2.RBAN SettingsOut) -> (
    then
    Output_Get_ComponentLinearAccScale_28_#1:Vector=~
    Output_Get_ComponentLinearVelScale_31_#2:Vector=~
    Output_Get_SimSpaceSettings_34_#3:Sim Space Settings Structure
   )
   n3 = GetOwnerSkeletalMeshComponent()
   n4 = BreakRotator(InRot=BreakTransform(InTransform=GetDeltaTransformFromRefPose(self=@n3, BoneName=spine_05, BaseName=spine_01)).Rotation)
   n5 = FInterpEaseInOut!(A:double=0, B:double=MapRangeClamped(Value=@n4.Yaw, InRangeA=0, InRangeB=-30, OutRangeB=40), Alpha:double=0, Exponent:double=0) -> (ReturnValue:double)
   Set BP Rotate(execute=@n1.then_1, BP Rotate:Rotator=MakeRotator(Roll=@n5, Pitch=@n5)) -> (Output_Get:Rotator)
   n6 = MapRangeClamped(Value=@n4.Yaw, InRangeA=0, InRangeB=-10, OutRangeB=-1.5)
   > Set BP Translate(BP Translate:Vector=MakeVector(Y=FInterpEaseInOut(A=0, B=@n6, Alpha=0, Exponent=0))) -> (Output_Get:Vector)
   Set Skydive(execute=@n1.then_2, Skydive:double="To Float (Boolean)"(InBool=Get bIsSkyDiving())) -> (Output_Get:double)
   > Set FLoat(FLoat:double="To Float (Boolean)"(InBool=Get bIsSkydiveFloating())) -> (Output_Get:double)
   > Set dive(dive:double="To Float (Boolean)"(InBool=Get bIsSkydiveDiveMode())) -> (Output_Get:double)
   n7 = Get bIsCrouchSprinting()
   > Set SprintNoCrouch(SprintNoCrouch:bool=AND Boolean(A=Get bIsSprinting(), B=NOT Boolean(A=@n7))) -> (Output_Get:bool)
   Event OnCharacterCustomizationLoaded()
   > n8 = Cast To FortPlayerPawn(Object=TryGetFortPawn())
   > n9 = WatchCosmeticStat!(
    execute:Exec
    Pawn:Fort Player Pawn Object Reference=@n8
    StatTag:Fort Stat Manager Tag Structure=(TagName="StatManager.OpenChests")
    DefaultValue:int=0
    bBroadcastInitialValue:bool=false
    InitialValueBroadcastDelay:float=-1
    ChangeValueBroadcastDelay:float=-1
    bConsumeWaitingNotifications:bool=true
   ) -> (then:Exec, OnStatChanged:Exec, StatName:Name, StatValue:int) {is=title K2Node_AsyncAction}
   Branch(execute=@n9.OnStatChanged, Condition="integer > integer"(A=@n9.StatValue))
   > n10 = Set Active_chest(Active_chest:bool=true) -> (Output_Get:bool)
   > Set Chest_opened(Chest_opened:double="To Float (Boolean)"(InBool=@n10)) -> (Output_Get:double)
   > Delay(Duration=3)
   > n11 = Set Active_chest(Active_chest=false)
   > Set Chest_opened(Chest_opened="To Float (Boolean)"(InBool=@n11))
   Branch(Condition=Get bIsInFrontEnd())
   > Do Once()
   > Delay(Duration=0)
   > Chestopen() {is=title K2Node_CallFunction}
   TryGetPawnOwner()
   Set Timer by Event(Time=4, bLooping=true)
   Sequence!(execute) -> (then_0, then_1)
  ]}
]
variables [
 Active_chest {type=boolean}
 BP Rotate {type=Rotator}
 BP Translate {type=Vector}
 Chest_opened {type=number}
 dive {type=number}
 FLoat {type=number}
 Gravity Override {type=Vector}
 Gravity Override Params {type=GravityOverrideParamsStruct}
 GravityOverride {type=Vector}
 open {type=number}
 open active {type=boolean}
 RBAN Params {
  type=RBANParamsStruct
  default {
   dBNO {simSpaceSettings {
     maxAngularAcceleration=250
     maxAngularVelocity=50
     maxLinearAcceleration=400
     maxLinearVelocity=100
     velocityScaleZ=0
     worldAlpha=0.35
     externalLinearDragV=<0.2 0.2 0.2>
    }}
   dBNO_Carried {simSpaceSettings {
     maxAngularAcceleration=250
     maxAngularVelocity=50
     maxLinearAcceleration=400
     maxLinearVelocity=100
     velocityScaleZ=0
     externalLinearDragV=<0.2 0.2 0.2>
    }}
   emoteOrMelee {simSpaceSettings {
     maxAngularAcceleration=250
     maxAngularVelocity=50
     maxLinearAcceleration=400
     maxLinearVelocity=100
     velocityScaleZ=0
     worldAlpha=0.35
     externalLinearDragV=<0.2 0.2 0.2>
    }}
   front_End {simSpaceSettings {
     maxAngularAcceleration=1000
     maxAngularVelocity=3
     maxLinearAcceleration=300
     maxLinearVelocity=700
     velocityScaleZ=0
     externalLinearDragV=<0.2 0.2 0.2>
    }}
   onGround_Crouch_Moving {simSpaceSettings {
     maxAngularAcceleration=250
     maxAngularVelocity=50
     maxLinearAcceleration=400
     maxLinearVelocity=100
     velocityScaleZ=0
     worldAlpha=0.35
     externalLinearDragV=<0.2 0.2 0.2>
    }}
   onGround_Falling {simSpaceSettings {
     maxAngularAcceleration=250
     maxAngularVelocity=50
     maxLinearAcceleration=400
     maxLinearVelocity=100
     velocityScaleZ=0
     worldAlpha=0.35
     externalLinearDragV=<0.2 0.2 0.2>
    }}
   onGround_Moving {simSpaceSettings {
     maxAngularAcceleration=250
     maxAngularVelocity=50
     maxLinearAcceleration=400
     maxLinearVelocity=100
     velocityScaleZ=0
     worldAlpha=0.35
     externalLinearDragV=<0.2 0.2 0.2>
    }}
   onGround_Standing {simSpaceSettings {
     maxAngularAcceleration=250
     maxAngularVelocity=50
     maxLinearAcceleration=400
     maxLinearVelocity=100
     velocityScaleZ=0
     worldAlpha=0.35
     externalLinearDragV=<0 0.5 0>
    }}
   rail_Grinding {simSpaceSettings {
     maxAngularAcceleration=250
     maxAngularVelocity=50
     maxLinearAcceleration=400
     maxLinearVelocity=100
     velocityScaleZ=0
     worldAlpha=0.35
     externalLinearDragV=<0.2 0.2 0.2>
    }}
   riding_Moving {simSpaceSettings {
     maxAngularAcceleration=250
     maxAngularVelocity=50
     maxLinearAcceleration=400
     maxLinearVelocity=100
     velocityScaleZ=0
     worldAlpha=0.35
     externalLinearDragV=<0.2 0.2 0.2>
    }}
   riding_Moving_Targeting {simSpaceSettings {
     maxAngularAcceleration=250
     maxAngularVelocity=50
     maxLinearAcceleration=400
     maxLinearVelocity=100
     velocityScaleZ=0
     worldAlpha=0.35
     externalLinearDragV=<0.2 0.2 0.2>
    }}
   riding_Standing {simSpaceSettings {
     maxAngularAcceleration=250
     maxAngularVelocity=50
     maxLinearAcceleration=400
     maxLinearVelocity=100
     velocityScaleZ=0
     worldAlpha=0.35
     externalLinearDragV=<0.2 0.2 0.2>
    }}
   skydive_DivingDown {simSpaceSettings {
     maxAngularAcceleration=250
     maxAngularVelocity=50
     maxLinearAcceleration=400
     maxLinearVelocity=100
     velocityScaleZ=0
     worldAlpha=0.35
     externalLinearDragV=<0.2 0.2 0.2>
    }}
   skydive_FloatingDown {
    componentLinearAccScale=<0 0.5 0>
    componentLinearVelScale=<0 0.5 0>
    simSpaceSettings {
     maxAngularAcceleration=0
     maxAngularVelocity=50
     maxLinearAcceleration=250
     maxLinearVelocity=100
     velocityScaleZ=0
     worldAlpha=0.35
     externalLinearDragV=<0.2 0.2 0.2>
    }
   }
   skydive_GoingUp {simSpaceSettings {
     maxAngularAcceleration=250
     maxAngularVelocity=50
     maxLinearAcceleration=400
     maxLinearVelocity=100
     velocityScaleZ=0
     worldAlpha=0.35
     externalLinearDragV=<0.2 0.2 0.2>
    }}
   skydive_Parachuting {
    componentLinearAccScale=<0 0.1 0.3>
    componentLinearVelScale=<0 0.1 0.3>
    simSpaceSettings {
     maxAngularAcceleration=250
     maxAngularVelocity=50
     maxLinearAcceleration=400
     maxLinearVelocity=100
     velocityScaleZ=0
     worldAlpha=0.35
     externalLinearDragV=<0.2 0.2 0.2>
    }
   }
   skydive_Surfing {simSpaceSettings {
     maxAngularAcceleration=250
     maxAngularVelocity=50
     maxLinearAcceleration=400
     maxLinearVelocity=100
     velocityScaleZ=0
     worldAlpha=0.35
     externalLinearDragV=<0.2 0.2 0.2>
    }}
   swimming {simSpaceSettings {
     maxAngularAcceleration=250
     maxAngularVelocity=50
     maxLinearAcceleration=400
     maxLinearVelocity=100
     velocityScaleZ=0
     worldAlpha=0.35
     externalLinearDragV=<0.2 0.2 0.2>
    }}
   tactical_Sprint {simSpaceSettings {
     maxAngularAcceleration=250
     maxAngularVelocity=50
     maxLinearAcceleration=400
     maxLinearVelocity=100
     velocityScaleZ=0
     worldAlpha=0.35
     externalLinearDragV=<0.2 0.2 0.2>
    }}
   ziplining {simSpaceSettings {
     maxAngularAcceleration=250
     maxAngularVelocity=50
     maxLinearAcceleration=400
     maxLinearVelocity=100
     velocityScaleZ=0
     worldAlpha=0.35
     externalLinearDragV=<0.2 0.2 0.2>
    }}
  }
 }
 RBAN Settings Out {
  type=RBANSettingStruct
  default {
   componentLinearAccScale=<0 0 0>
   componentLinearVelScale=<0 0 0>
   simSpaceSettings {
    dampingAlpha=1
    maxAngularAcceleration=250
    maxAngularVelocity=50
    maxLinearAcceleration=400
    maxLinearVelocity=100
    velocityScaleZ=0
    worldAlpha=0.35
    externalAngularVelocity=<0 0 0>
    externalLinearDragV=<0.2 0.2 0.2>
    externalLinearVelocity=<0 0 0>
   }
  }
 }
 Sim Space Settings {
  type=SimSpaceSettings
  default {
   dampingAlpha=1
   maxAngularAcceleration=250
   maxAngularVelocity=50
   maxLinearAcceleration=1000
   maxLinearVelocity=100
   velocityScaleZ=0
   worldAlpha=0.4
   externalLinearDragV=<0.2 0.2 0.1>
  }
 }
 Skel_Mesh {type=SkeletalMeshComponent}
 Skydive {type=number}
 SprintNoCrouch {type=boolean}
]
